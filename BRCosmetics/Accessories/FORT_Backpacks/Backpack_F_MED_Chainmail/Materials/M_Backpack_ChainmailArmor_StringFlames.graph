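# Unlit translucent two-sided material for wispy string flames on the Chainmail backpack, built for Niagara mesh particles.
# A UV-shaped mask combines with panning noise, warped by two multiplied Voronoi samples and offset by a particle dynamic parameter.
# The result drives curve-atlas color, tinted by particle color and time of day, and edge-hardened opacity with depth fade on deferred.
# Particle-driven world position offset adds upward push and clamped positional lag, weighted along U.
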Material /BRCosmetics/Accessories/FORT_Backpacks/Backpack_F_MED_Chainmail/Materials/M_Backpack_ChainmailArmor_StringFlames
graph {
 properties {
  bUsedWithNiagaraMeshParticles=true
  blendMode=BLEND_Translucent
  pixelDepthOffsetMode=PDOM_Legacy
  shadingModel=MSM_Unlit
  twoSided=true
  usedShadingModels=Unlit
 }
 nodes [
  n1 = ComponentMask(None=TextureCoordinate()) {g=true}
  n2 = Constant() {r=0.5}
  n3 = Multiply(
   A=ComponentMask(None=TextureCoordinate()) {r=true}
   B=Power(
    Base=SphereMask(A=@n1, B=@n2) {attenuationRadius=0.5, hardnessPercent=0}
    Exp=ScalarParameter() {defaultValue=1, parameterName=Concave, sliderMax=10}
   )
  )
  n4 = ComponentMask(None=TextureCoordinate()) {r=true}
  n5 = Power(Base=@n4) {constExponent=10}
  LinearInterpolate(A=@n3, Alpha=@n5) {constB=0}
  n6 = SmoothStep(Value=Subtract(A=@n3, B=@n5)) {constMax=0.9}
  n7 = TextureCoordinate()
  n8 = Multiply(
   A=NamedRerouteUsage()
   B=ScalarParameter() {defaultValue=0.1, group=Wispy Flame, parameterName=Wispy Distortion Strength, sliderMax=1, sliderMin=-1}
  ) {constB=0.1}
  n9 = ScalarParameter() {defaultValue=1, group=Wispy Flame, parameterName=Speed X, sliderMax=5, sliderMin=-5}
  n10 = ScalarParameter() {group=Wispy Flame, parameterName=Speed Y, sliderMax=5, sliderMin=-5}
  n11 = Multiply(
   A=Panner(Coordinate=Add(A=@n7, B=@n8), Speed=AppendVector(A=@n9, B=@n10)) {speedX=1}
   B=AppendVector(
    A=ScalarParameter() {defaultValue=1, group=Wispy Flame, parameterName=Scale X, sliderMax=5}
    B=ScalarParameter() {defaultValue=1, group=Wispy Flame, parameterName=Scale Y, sliderMax=5}
   )
  )
  n12 = AppendVector(A=DynamicParameter() {paramNames [U Offset, WPO Direction, Current Position Y, Current Position Z]}.U Offset, B=Constant())
  n13 = TextureSampleParameter2D(UVs=MaterialFunctionCall(XY=Add(A=@n11, B=@n12)) {materialFunction=&Swizzle}.YX) {
   group=Wispy Flame
   parameterName=Flame Texture
   texture=/BRCosmetics/Athena/Effects/Particles/Skydiving/NS_Contrail_Chainmail/T_WispyFireNoise03
  }
  n14 = LinearInterpolate(A=Multiply(A=Multiply(A=@n6) {constB=2}, B=@n13.R), B=@n6, Alpha=@n6)
  n15 = Saturate(None=@n14)
  n16 = ScalarParameter() {defaultValue=1, group=Color, parameterName=Color Power, sliderMax=5}
  n17 = CurveAtlasRowParameter(CurveTime=SmoothStep(
    Min=ScalarParameter() {group=Color, parameterName=Color Smooth Step Min, sliderMax=1, sliderMin=-1}
    Max=ScalarParameter() {defaultValue=2, group=Color, parameterName=Color Smooth Step Max, sliderMax=5}
    Value=Power(Base=@n15, Exp=@n16) {constExponent=1}
   ) {constMax=2}) {
   atlas=/Game/Athena/LinearColorCurves/CurveAtlas_Chainmail
   curve=/Game/Athena/LinearColorCurves/Curve_ChainMail_GradientForMesh
   defaultValue=2
   parameterName=Curve_Color
  }
  n18 = ParticleColor()
  n19 = VectorParameter() {parameterName=TOD color, defaultValue=<rgba 1 1 1 0.5>}
  MaterialFunctionCall(Emissive=Multiply(
    A=Multiply(A=@n17, B=@n18.RGB)
    B=MaterialFunctionCall(Value_Morning=@n19.R, Value_Day=@n19.R, Value_Evening=@n19.R, Value_Night=@n19.R) {materialFunction=&Scalar_Time_of_Day}.Result
   )) {materialFunction=&MF_FX_EmissiveAdjustments}
  n20 = SmoothStep(Max=OneMinus(None=ScalarParameter() {defaultValue=0.9, parameterName=Edge Hardness, sliderMax=1}), Value=Multiply(A=@n18.A, B=@n14)) {constMax=0.1}
  n21 = DepthFade(Opacity=@n20, FadeDistance=ScalarParameter() {defaultValue=35, parameterName=Depth Fade Distance, sliderMax=100}) {fadeDistanceDefault=25}
  Multiply(
   A=ShadingPathSwitch(Default=@n21, Deferred=@n21, Forward=@n20, Mobile=@n20)
   B=ScalarParameter() {defaultValue=1, parameterName=Opacity, sliderMax=1}
  )
  n22 = DynamicParameter() {paramNames [U Offset, WPO Direction, Current Position Y, Current Position Z]}
  n23 = Constant3Vector() {constant=<rgba 0 1 0 0>}
  n24 = Multiply(
   A=Transform(None=Multiply(A=@n22.WPO Direction, B=@n23) {constA=1}) {transformSourceType=TRANSFORMSOURCE_Local}
   B=ScalarParameter() {defaultValue=3, parameterName=WPO Upwards, sliderMax=50}
  )
  n25 = DynamicParameter() {parameterIndex=1, paramNames [Position X, Position Y, Position Z, Param4]}
  n26 = MaterialFunctionCall(X=@n25.Position X, Y=@n25.Position X, Z=@n25.Position X) {materialFunction=&MakeFloat3}
  n27 = ParticlePositionWS()
  n28 = ScalarParameter() {defaultValue=20, parameterName=WPO Max Offset, sliderMax=50}
  n29 = Multiply(
   A=VectorParameter() {parameterName=WPO Axis Mask, defaultValue=<rgba 1 1 1 1>}.RGB
   B=Clamp(None=Subtract(A=@n26.Result, B=@n27), Min=Multiply(A=@n28) {constB=-1}, Max=@n28)
  )
  n30 = OneMinus(None=ComponentMask(None=TextureCoordinate()) {r=true})
  n31 = ScalarParameter() {defaultValue=1, parameterName=WPO Bend, sliderMax=10}
  Multiply(
   A=ScalarParameter() {parameterName=WPO Intensity, sliderMax=100}
   B=LinearInterpolate(B=Add(A=@n24, B=@n29), Alpha=Power(Base=@n30, Exp=@n31))
  )
  n32 = DynamicParameter() {paramNames [U Offset, WPO Direction, Current Position Y, Current Position Z]}
  n33 = Constant()
  n34 = Panner(
   Coordinate=TextureCoordinate()
   Speed=AppendVector(
    A=ScalarParameter() {group=UV Distortion, parameterName=UV Speed X 2, sliderMax=5, sliderMin=-5}
    B=ScalarParameter() {defaultValue=1, group=UV Distortion, parameterName=UV Speed Y 2, sliderMax=5, sliderMin=-5}
   )
  ) {speedX=-0.5}
  n35 = AppendVector(
   A=ScalarParameter() {defaultValue=1, group=UV Distortion, parameterName=UV Scale X 2, sliderMax=5}
   B=ScalarParameter() {defaultValue=1, group=UV Distortion, parameterName=UV Scale Y 2, sliderMax=5}
  )
  n36 = TextureSampleParameter2D(UVs=Add(A=AppendVector(A=@n32.U Offset, B=@n33), B=Multiply(A=@n34, B=@n35))) {
   group=UV Distortion
   parameterName=Distortion Texture
   texture=/BRCosmetics/Athena/Effects/Particles/Skydiving/NS_Contrail_PinkSpike/Textures/T_PinkSpike_Voronoi2D_Height
  }
  n37 = DynamicParameter() {paramNames [U Offset, WPO Direction, Current Position Y, Current Position Z]}
  n38 = Constant()
  n39 = Panner(
   Coordinate=TextureCoordinate()
   Speed=AppendVector(
    A=ScalarParameter() {group=UV Distortion, parameterName=UV Speed X, sliderMax=5, sliderMin=-5}
    B=ScalarParameter() {defaultValue=0.2, group=UV Distortion, parameterName=UV Speed Y, sliderMax=5, sliderMin=-5}
   )
  ) {speedX=-0.5}
  n40 = AppendVector(
   A=ScalarParameter() {defaultValue=0.7, group=UV Distortion, parameterName=UV Scale X, sliderMax=5}
   B=ScalarParameter() {defaultValue=0.8, group=UV Distortion, parameterName=UV Scale Y, sliderMax=5}
  )
  n41 = TextureSampleParameter2D(UVs=Add(A=AppendVector(A=@n37.U Offset, B=@n38), B=Multiply(A=@n39, B=@n40))) {
   group=UV Distortion
   parameterName=Distortion Texture
   texture=&T_PinkSpike_Voronoi2D_Height
  }
  n42 = Multiply(A=@n36.R, B=@n41.R)
  n43 = Multiply(A=@n36.G, B=@n41.G)
  NamedRerouteDeclaration(None=Subtract(A=MaterialFunctionCall(X=@n42, Y=@n43) {materialFunction=&MakeFloat2}.Result) {constB=0.25}) {name=Distortion, nodeColor=<rgba 1 0.3281 0 1>}
  OneMinus(None=@n4)
 ]
}
