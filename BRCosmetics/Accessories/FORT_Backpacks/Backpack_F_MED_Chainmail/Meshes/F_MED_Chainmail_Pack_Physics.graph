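# Physics asset for the F_MED_Chainmail backpack cosmetic, giving its two clubs, their chains and the left and right mitt strings capsule bodies and joints.
# Top and bottom clubs hang off the pack with narrow limits, each feeding a twist-only master joint into a chain of wider-swinging links.
# Both seven-link mitt strings start at master bodies with free angular motion on the pack, which is itself angularly free relative to root.
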PhysicsAsset /BRCosmetics/Accessories/FORT_Backpacks/Backpack_F_MED_Chainmail/Meshes/F_MED_Chainmail_Pack_Physics
bodies {
 dyn_ClubBt1 [{shapeType=Capsule, length=14.159, radius=2, shapeName=dyn_ClubBt1_capsule, center=<2.0943 0 0.0228>, rotation=<rot 90 0 0>}]
 dyn_ClubTp1 [{shapeType=Capsule, length=14.159, radius=2, shapeName=dyn_ClubTp1_capsule, center=<4.9465 0 0.0228>, rotation=<rot 90 0 0>}]
 dyn_clubBtChain1 [{shapeType=Capsule, length=6, radius=6, shapeName=dyn_clubBtChain1_capsule, center=<15 0 0>, rotation=<rot 90 0 0>}]
 dyn_clubBtChain2 [{shapeType=Capsule, length=1.23, radius=1.33, shapeName=dyn_clubBtChain2_capsule, center=<1.35 0 0>, rotation=<rot 90 0 0>}]
 dyn_clubBtChain3 [{shapeType=Capsule, length=0.66, radius=1.34, shapeName=dyn_clubBtChain3_capsule, center=<1.27 0 0>, rotation=<rot 90 0 0>}]
 dyn_clubBtChain4 [{shapeType=Capsule, length=1.2, radius=0.95, shapeName=dyn_clubBtChain4_capsule, center=<1.15 0 0>, rotation=<rot 90 0 0>}]
 dyn_clubBtChain5 [{shapeType=Capsule, length=1.38, radius=1.07, shapeName=dyn_clubBtChain5_capsule, center=<1.48 0 0>, rotation=<rot 90 0 0>}]
 dyn_clubBtChain6 [{shapeType=Capsule, length=5.0026, radius=1.32, shapeName=dyn_clubBtChain6_capsule, center=<0.78 -0.3 0>, rotation=<rot 90 27 75.9>}]
 dyn_clubBtChain_MSTR [{
   shapeType=Capsule
   length=12.8515
   radius=6.0578
   shapeName=dyn_clubBtChain_MSTR_capsule
   center=<6.3989 0.0001 0.0482>
   rotation=<rot 0 -89.9993 -90.4313>
  }]
 dyn_clubTpChain1 [{shapeType=Capsule, length=0.1798, radius=2.47, shapeName=dyn_clubTpChain1_capsule, center=<1.95 0 0>, rotation=<rot 90 0 0>}]
 dyn_clubTpChain2 [{shapeType=Capsule, length=0.44, radius=1.4, shapeName=dyn_clubTpChain2_capsule, center=<1.4275 0 0>, rotation=<rot 90 0 0>}]
 dyn_clubTpChain3 [{shapeType=Capsule, length=1.3, radius=1.1, shapeName=dyn_clubTpChain3_capsule, center=<1.34 0 0>, rotation=<rot 90 0 0>}]
 dyn_clubTpChain4 [{shapeType=Capsule, length=1.4, radius=1, shapeName=dyn_clubTpChain4_capsule, center=<1.21 0 0>, rotation=<rot 90 0 0>}]
 dyn_clubTpChain5 [{shapeType=Capsule, length=2.4, radius=2.51, shapeName=dyn_clubTpChain5_capsule, center=<3.23 0 -0.47>, rotation=<rot 90 0 0>}]
 dyn_clubTpChain_MSTR [{shapeType=Capsule, length=1, radius=1, shapeName=dyn_clubTpChain_MSTR_capsule, rotation=<rot 0 -90 0>}]
 dyn_mittStringLf1 [{shapeType=Capsule, length=2.8, radius=0.5, shapeName=dyn_mittStringLf1_capsule, center=<1.27 0 -2.4>, rotation=<rot 90 0 0>}]
 dyn_mittStringLf2 [{shapeType=Capsule, length=2, radius=0.5, shapeName=dyn_mittStringLf2_capsule, center=<1.6 0 0>, rotation=<rot 90 0 0>}]
 dyn_mittStringLf3 [{shapeType=Capsule, length=1.57, radius=0.5, shapeName=dyn_mittStringLf3_capsule, center=<1.4 0 0>, rotation=<rot 90 0 0>}]
 dyn_mittStringLf4 [{shapeType=Capsule, length=1.55, radius=0.5, shapeName=dyn_mittStringLf4_capsule, center=<1.2 0 0>, rotation=<rot 90 0 0>}]
 dyn_mittStringLf5 [{shapeType=Capsule, length=1.2, radius=0.5, shapeName=dyn_mittStringLf5_capsule, center=<1.2 0 0>, rotation=<rot 90 0 0>}]
 dyn_mittStringLf6 [{shapeType=Capsule, length=1.2, radius=0.5, shapeName=dyn_mittStringLf6_capsule, center=<1 0 0>, rotation=<rot 90 0 0>}]
 dyn_mittStringLf7 [{shapeType=Capsule, length=0.5, radius=0.5, shapeName=dyn_mittStringLf7_capsule, center=<0.72 0 0>, rotation=<rot 90 0 0>}]
 dyn_mittStringLf_MSTR [{shapeType=Capsule, length=1, radius=1, shapeName=dyn_mittStringLf_MSTR_capsule, rotation=<rot 0 -90 0>}]
 dyn_mittStringRt1 [{shapeType=Capsule, length=2.6, radius=0.5, shapeName=dyn_mittStringRt1_capsule, center=<1.4 0 -2.25>, rotation=<rot 90 0 0>}]
 dyn_mittStringRt2 [{shapeType=Capsule, length=2, radius=0.5, shapeName=dyn_mittStringRt2_capsule, center=<1.3 0 -1>, rotation=<rot 90 0 0>}]
 dyn_mittStringRt3 [{shapeType=Capsule, length=2, radius=0.5, shapeName=dyn_mittStringRt3_capsule, center=<1.35 0 0>, rotation=<rot 90 0 0>}]
 dyn_mittStringRt4 [{shapeType=Capsule, length=1.85, radius=0.5, shapeName=dyn_mittStringRt4_capsule, center=<1.2 0 0>, rotation=<rot 90 0 0>}]
 dyn_mittStringRt5 [{shapeType=Capsule, length=1.6, radius=0.5, shapeName=dyn_mittStringRt5_capsule, center=<1.2 0 0>, rotation=<rot 90 0 0>}]
 dyn_mittStringRt6 [{shapeType=Capsule, length=1.63, radius=0.5, shapeName=dyn_mittStringRt6_capsule, center=<1 0 0>, rotation=<rot 90 0 0>}]
 dyn_mittStringRt7 [{shapeType=Capsule, length=0.71, radius=0.5, shapeName=dyn_mittStringRt7_capsule, center=<0.7 0 0>, rotation=<rot 90 0 0>}]
 dyn_mittStringRt_MSTR [{shapeType=Capsule, length=1, radius=1, shapeName=dyn_mittStringRt_MSTR_capsule, rotation=<rot 0 -90 0>}]
 pack [
  {shapeType=Capsule, length=5.7, radius=5.6, shapeName=pack_capsule, center=<-5.1 -4.7 -7>, rotation=<rot -43 0 -7.62>}
  {shapeType=Capsule, length=6.2, radius=8.5, shapeName=pack_capsule, center=<-1 -4.75 -2.8>, rotation=<rot -45 3 -25>}
 ]
 root [{shapeType=Capsule, length=2, radius=2, shapeName=root_capsule}]
}
constraints [
 {
  bone1Name=pack
  bone2Name=root
  swing1LimitDegrees=45
  swing1Motion=Free
  swing2LimitDegrees=45
  swing2Motion=Free
  twistLimitDegrees=45
  twistMotion=Free
 }
 {bone1Name=dyn_ClubBt1, bone2Name=pack, swing1LimitDegrees=20, swing2LimitDegrees=20, twistLimitDegrees=10}
 {bone1Name=dyn_clubBtChain1, bone2Name=dyn_clubBtChain_MSTR, swing1LimitDegrees=20, swing2LimitDegrees=0, twistLimitDegrees=0}
 {bone1Name=dyn_clubBtChain2, bone2Name=dyn_clubBtChain1, swing1LimitDegrees=50, swing2LimitDegrees=50, twistLimitDegrees=10}
 {bone1Name=dyn_clubBtChain3, bone2Name=dyn_clubBtChain2, swing1LimitDegrees=50, swing2LimitDegrees=50, twistLimitDegrees=10}
 {bone1Name=dyn_clubBtChain4, bone2Name=dyn_clubBtChain3, swing1LimitDegrees=50, swing2LimitDegrees=50, twistLimitDegrees=10}
 {bone1Name=dyn_clubBtChain5, bone2Name=dyn_clubBtChain4, swing1LimitDegrees=50, swing2LimitDegrees=50, twistLimitDegrees=10}
 {bone1Name=dyn_clubBtChain6, bone2Name=dyn_clubBtChain5, swing1LimitDegrees=50, swing2LimitDegrees=50, twistLimitDegrees=10}
 {bone1Name=dyn_ClubTp1, bone2Name=pack, swing1LimitDegrees=20, swing2LimitDegrees=20, twistLimitDegrees=10}
 {bone1Name=dyn_clubTpChain1, bone2Name=dyn_clubTpChain_MSTR, swing1LimitDegrees=20, swing2LimitDegrees=0, twistLimitDegrees=0}
 {bone1Name=dyn_clubTpChain2, bone2Name=dyn_clubTpChain1, swing1LimitDegrees=50, swing2LimitDegrees=50, twistLimitDegrees=10}
 {bone1Name=dyn_clubTpChain3, bone2Name=dyn_clubTpChain2, swing1LimitDegrees=50, swing2LimitDegrees=50, twistLimitDegrees=10}
 {bone1Name=dyn_clubTpChain4, bone2Name=dyn_clubTpChain3, swing1LimitDegrees=50, swing2LimitDegrees=50, twistLimitDegrees=10}
 {bone1Name=dyn_clubTpChain5, bone2Name=dyn_clubTpChain4, swing1LimitDegrees=50, swing2LimitDegrees=50, twistLimitDegrees=10}
 {bone1Name=dyn_mittStringLf1, bone2Name=dyn_mittStringLf_MSTR, swing1LimitDegrees=30, swing2LimitDegrees=30, twistLimitDegrees=0}
 {bone1Name=dyn_mittStringLf2, bone2Name=dyn_mittStringLf1, swing1LimitDegrees=25, swing2LimitDegrees=25, twistLimitDegrees=10}
 {bone1Name=dyn_mittStringLf3, bone2Name=dyn_mittStringLf2, swing1LimitDegrees=25, swing2LimitDegrees=25, twistLimitDegrees=10}
 {bone1Name=dyn_mittStringLf4, bone2Name=dyn_mittStringLf3, swing1LimitDegrees=25, swing2LimitDegrees=25, twistLimitDegrees=10}
 {bone1Name=dyn_mittStringLf5, bone2Name=dyn_mittStringLf4, swing1LimitDegrees=25, swing2LimitDegrees=25, twistLimitDegrees=10}
 {bone1Name=dyn_mittStringLf6, bone2Name=dyn_mittStringLf5, swing1LimitDegrees=25, swing2LimitDegrees=25, twistLimitDegrees=10}
 {bone1Name=dyn_mittStringLf7, bone2Name=dyn_mittStringLf6, swing1LimitDegrees=25, swing2LimitDegrees=25, twistLimitDegrees=10}
 {bone1Name=dyn_mittStringRt1, bone2Name=dyn_mittStringRt_MSTR, swing1LimitDegrees=30, swing2LimitDegrees=30, twistLimitDegrees=0}
 {bone1Name=dyn_mittStringRt2, bone2Name=dyn_mittStringRt1, swing1LimitDegrees=25, swing2LimitDegrees=25, twistLimitDegrees=10}
 {bone1Name=dyn_mittStringRt3, bone2Name=dyn_mittStringRt2, swing1LimitDegrees=25, swing2LimitDegrees=25, twistLimitDegrees=10}
 {bone1Name=dyn_mittStringRt4, bone2Name=dyn_mittStringRt3, swing1LimitDegrees=25, swing2LimitDegrees=25, twistLimitDegrees=10}
 {bone1Name=dyn_mittStringRt5, bone2Name=dyn_mittStringRt4, swing1LimitDegrees=25, swing2LimitDegrees=25, twistLimitDegrees=10}
 {bone1Name=dyn_mittStringRt6, bone2Name=dyn_mittStringRt5, swing1LimitDegrees=25, swing2LimitDegrees=25, twistLimitDegrees=10}
 {bone1Name=dyn_mittStringRt7, bone2Name=dyn_mittStringRt6, swing1LimitDegrees=25, swing2LimitDegrees=25, twistLimitDegrees=10}
 {bone1Name=dyn_clubBtChain_MSTR, bone2Name=dyn_ClubBt1, swing1LimitDegrees=0, swing2LimitDegrees=0, twistLimitDegrees=45}
 {
  bone1Name=dyn_mittStringRt_MSTR
  bone2Name=pack
  swing1LimitDegrees=45
  swing1Motion=Free
  swing2LimitDegrees=45
  swing2Motion=Free
  twistLimitDegrees=45
  twistMotion=Free
 }
 {
  bone1Name=dyn_mittStringLf_MSTR
  bone2Name=pack
  swing1LimitDegrees=45
  swing1Motion=Free
  swing2LimitDegrees=45
  swing2Motion=Free
  twistLimitDegrees=45
  twistMotion=Free
 }
 {bone1Name=dyn_clubTpChain_MSTR, bone2Name=dyn_ClubTp1, swing1LimitDegrees=0, swing2LimitDegrees=0, twistLimitDegrees=45}
]
